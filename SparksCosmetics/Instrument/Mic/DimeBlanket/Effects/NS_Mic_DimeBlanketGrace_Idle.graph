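# Idle Niagara effect for the DimeBlanketGrace mic cosmetic, scattering occasional short-lived sparkles over the instrument.
# One emitter randomly places particles along two tapered handle segments or on a ring, with spawn probability depending on System.Music_Reactive_FX_Preview.
# That value also boosts alpha when User.MusicReactive is off.
# Three sprite renderers share flare colors blended by User.Variant, with quality, time-of-day, front-end preview and distance adjustments.
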NiagaraSystem /SparksCosmetics/Instrument/Mic/DimeBlanket/Effects/NS_Mic_DimeBlanketGrace_Idle
dependencies {
 usedDataInterfaces [&NiagaraDataInterfaceCurve, &NiagaraDataInterfaceArrayFloat, &NiagaraDataInterfaceColorCurve]
 usedDynamicInputs [
  &Multiply_Vector2D_ByFloat
  &MakeVector2D
  &ColorFromCurve
 ]
 usedModules [
  &SystemState
  &self:SystemSpawnScript.SystemScriptSource.SystemScriptGraph.NiagaraNodeAssignment_4.SetVariables_#1
  &self:SystemSpawnScript.SystemScriptSource.SystemScriptGraph.NiagaraNodeAssignment_6.SetVariables_#2
 ]
}
emitters {Sparkles {
  emitterUpdateScript [
   &EmitterState()
   &SpawnRate(
    Spawn Probability=&MakeCustomFloatFromBool(Bool=&SetBoolByFloatComparison(A=$System.Music_Reactive_FX_Preview, B=0), True Float=1, False Float=0.125)
    SpawnRate=0.75
    Use Spawn Probability=true
   )
  ]
  particleSpawnScript [
   /SetVariables(Transient.LengthPos:NiagaraFloat=&RandomRangeFloat(Minimum=0, Maximum=1)) {isSetParameters=true}
   /SetVariables(
    Transient.HandlePosA:Vector3f=&Add_Vector(
     A=&Multiply_Vector(
      A=<1 1 0>
      B=&RandomVector(Vector Scale=&FloatFromCurve(FloatCurve=DataInterface() {propertyValues {Curve {keys [<at 0 1.5>,<at 0.5 2>]}}}, CurveIndex=$Transient.LengthPos))
     )
     B=&Lerp_Vector(A=<0 0 -15.5>, B=<0 0 -11.5>, Alpha=$Transient.LengthPos)
    )
    Transient.HandlePosB:Vector3f=&Add_Vector(
     A=&Multiply_Vector(
      A=<1 1 0>
      B=&RandomVector(Vector Scale=&FloatFromCurve(FloatCurve=DataInterface() {propertyValues {Curve {keys [<at 0.5 2>,<at 1 2.5>]}}}, CurveIndex=$Transient.LengthPos))
     )
     B=&Lerp_Vector(A=<0 0 -2.5>, B=<0 0 4>, Alpha=$Transient.LengthPos)
    )
    Transient.RingPos:Vector3f=&Add_Vector(
     A=<0 0 6.5>
     B=&MakeVector(
      X=&Sine(Normalized Angle=$Transient.LengthPos, Period=1, Scale=3.6, Bias=0)
      Y=&Cosine(Normalized Angle=$Transient.LengthPos, Period=1, Scale=3.6, Bias=0)
      Z=0
     )
    )
   ) {isSetParameters=true}
   &InitializeParticle(
    AdjustAlpha=false
    AdjustHue=true
    AdjustSaturation=true
    AdjustValue=true
    Color=<rgba 2 2 2 1>
    Color Mode=::NewEnumerator3
    Hue Shift Range:Vector2f=<-0.05 0.05>
    Lifetime Max=1
    Lifetime Min=0.5
    Lifetime Mode=::NewEnumerator1
    Position Offset=&Lerp_Vector(
     A=&Lerp_Vector(
      A=$Transient.HandlePosA
      B=$Transient.HandlePosB
      Alpha=&UniformAOrBFloat(A=0, Evaluation Type=::NewEnumerator0, Recalculate Random Each Loop=true, Override Seed=false)
     )
     B=$Transient.RingPos
     Alpha=&UniformAOrBFloat(A=0, Distribution Weight=0.2, Evaluation Type=::NewEnumerator0, Recalculate Random Each Loop=true, Override Seed=false)
    )
    Sprite Size Mode=::NewEnumerator3
    Uniform Sprite Size=&Multiply_Float(
     A=&RandomRangeFloat(Minimum=12, Maximum=18)
     B=&SelectFloatFromArray(Direct Array Index=$Engine.QualityLevel, Float Selection Array=DataInterface() {propertyValues {FloatData [0.5,0.6,0.8,1,1]}})
    )
    UsePositionOffset=true
    Write Material Random=true
   )
   &CameraOffset(Camera Offset Amount=3)
  ]
  particleUpdateScript [
   &ParticleState()
   /SetVariables(Transient.BeatReactiveMult:NiagaraFloat=&MakeCustomFloatFromBool(
     Bool=$User.MusicReactive
     True Float=1
     False Float=&MakeCustomFloatFromBool(
      Bool=&SetBoolByFloatComparison(A=$System.Music_Reactive_FX_Preview, B=0)
      True Float=&Lerp_Float(
       A=&MakeCustomFloatFromBool(Bool=$User.bIsFrontEndPreview, True Float=1, False Float=1.4)
       B=&MakeCustomFloatFromBool(Bool=$User.bIsFrontEndPreview, True Float=1.75, False Float=2)
       Alpha=$System.Music_Reactive_FX_Preview
      )
      False Float=1
     )
    )) {isSetParameters=true}
   &ScaleColor(
    Scale Alpha=&Multiply_Float(
     A=$Transient.BeatReactiveMult
     B=&Multiply_Float(
      A=&FloatFromCurve(
       FloatCurve=DataInterface() {propertyValues {Curve {keys [
           {arriveTangent=6.5707, interpMode=RCIM_Cubic, leaveTangent=6.5707, tangentMode=RCTM_User, time=0, value=0}
           {interpMode=RCIM_Cubic, tangentMode=RCTM_User, time=0.3, value=1}
           {tangentMode=RCTM_User, time=1, value=0}
          ]}}}
       Scale Curve=&RandomRangeFloat(Minimum=0.1, Maximum=0.5)
      )
      B=&MakeFloatFromLinearColor(LinearColor=$System.TODColorMultiplier, Channel:ENiagaraLinearColor_Channels=::NewEnumerator3)
     )
    )
    Scale RGB=&Multiply_Vector(
     A=&VectorFromFloat(Value=&Multiply_Float(
       A=&RandomRangeFloat(Minimum=0.75, Maximum=1)
       B=&SelectFloatFromArray(Direct Array Index=$Engine.QualityLevel, Float Selection Array=DataInterface() {propertyValues {FloatData [10,5,1,1,1]}})
      ))
     B=&MakeVectorFromLinearColorRGB(Color=$System.TODColorMultiplier)
    )
    ScaleA=true
    ScaleRGB=true
   ) {module=ScaleColor001}
   &ScaleSpriteSize(
    Uniform Curve Scale=&MakeCustomFloatFromBool(
     Bool=$User.bIsFrontEndPreview
     True Float=1
     False Float=&FloatFromCurve(
      FloatCurve=DataInterface() {propertyValues {Curve {keys [<at 0 1>,<at 1000 2>]}}}
      CurveIndex=$Engine.Owner.LODDistance
      Scale Curve=&FNTimeOfDayScalar(Night Scale=1.25, Morning Scale=1, Evening Scale=1)
     )
    )
    Uniform Curve Sprite Scale=DataInterface() {propertyValues {Curve {keys [
        {interpMode=RCIM_Cubic, tangentMode=RCTM_User, time=0, value=0}
        {interpMode=RCIM_Cubic, tangentMode=RCTM_User, time=0.5, value=1}
        {tangentMode=RCTM_User, time=1, value=0}
       ]}}}
   )
   /SetVariables(
    Particles.CenterFlareColor:LinearColor=&MultiplyLinearColor(
     Linear Color=$Particles.Color
     Scale Factor=&MakeVector4_FromLinearColor(LinearColor=&LerpLinearColors(StartColor=<rgba 5 4.8775 3.0696 1>, EndColor=<rgba 3.0348 3.9019 5 1>, LerpFactor=$User.Variant))
    )
    Particles.CenterFlareSize:Vector2f=&Multiply_Vector2(A=$Particles.SpriteSize, B=&Vector2DFromFloat(Value=0.5))
    Particles.CoronaFlareColor:LinearColor=&MultiplyLinearColor(
     Linear Color=$Particles.Color
     Scale Factor=&MakeVector4_FromLinearColor(LinearColor=&LerpLinearColors(StartColor=<rgba 0.15 0.15 0.15 1>, EndColor=<rgba 0.0729 0.0729 0.0729 1>, LerpFactor=$User.Variant))
    )
    Particles.CoronaFlareSize:Vector2f=&Multiply_Vector2(A=$Particles.Initial.SpriteSize, B=&Vector2DFromFloat(Value=0.75))
   ) {isSetParameters=true}
  ]
  rendererClasses [&NiagaraSpriteRendererProperties]
  renderers [
   {rendererClass=&NiagaraSpriteRendererProperties}
   {rendererClass=&NiagaraSpriteRendererProperties, rendererIndex=1}
   {rendererClass=&NiagaraSpriteRendererProperties, rendererIndex=2}
  ]
 }}
summary {
 systemName=NS_Mic_DimeBlanketGrace_Idle
 emitters [{emitterName=Sparkles, rendererClasses [&NiagaraSpriteRendererProperties]}]
 userVariables [
  User.bIsFrontEndPreview {type=NiagaraBool}
  User.MusicReactive {type=NiagaraBool, defaultValue=true}
  User.Variant {type=NiagaraFloat, defaultValue=0, description=""}
 ]
}
